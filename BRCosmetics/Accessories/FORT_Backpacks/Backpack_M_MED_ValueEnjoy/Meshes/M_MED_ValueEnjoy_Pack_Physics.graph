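# Physics asset for the ValueEnjoy backpack cosmetic, providing bodies and joints for its left and right lower tubes.
# Each side has a small capsule on the base bone and an offset, rotated capsule on the dyn bone, joined by one constraint.
# Both constraints lock twist, while capsule shapes and first swing limit angles differ between left and right.
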PhysicsAsset /BRCosmetics/Accessories/FORT_Backpacks/Backpack_M_MED_ValueEnjoy/Meshes/M_MED_ValueEnjoy_Pack_Physics
bodies {
 base_lowerTube_l [{shapeType=Capsule, length=1, radius=1}]
 base_lowerTube_r [{shapeType=Capsule, length=1, radius=1}]
 dyn_lowerTube_l [{shapeType=Capsule, length=2.9511, radius=2.5395, center=<8.7131 -0.045 -0.1338>, rotation=<rot 71.4014 90.0001 90>}]
 dyn_lowerTube_r [{shapeType=Capsule, length=0.0557, radius=4.0122, center=<8.7321 -0.045 -0.1338>, rotation=<rot -82.6999 90.0001 -90.0001>}]
}
constraints [
 {
  bone1Name=dyn_lowerTube_l
  bone2Name=base_lowerTube_l
  swing1LimitDegrees=22.5
  swing2LimitDegrees=30
  twistLimitDegrees=45
  twistMotion=Locked
 }
 {
  bone1Name=dyn_lowerTube_r
  bone2Name=base_lowerTube_r
  swing1LimitDegrees=15
  swing2LimitDegrees=30
  twistLimitDegrees=45
  twistMotion=Locked
 }
]
